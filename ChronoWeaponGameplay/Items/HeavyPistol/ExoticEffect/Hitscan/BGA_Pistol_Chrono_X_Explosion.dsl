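; Blueprint actor for the Chrono heavy pistol exotic hitscan effect, an explosion spawned in the world.
; On BeginPlay it plays its ExplosionSound variable as a one-shot sound at the actor's location.
; The construction script is empty and the graph holds no damage or visual logic.
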(event EventBeginPlay
  (Audio|PlaySoundatLocation (Variables|Default|GetExplosionSound) (Transformation|GetActorLocation)))

(fn ConstructionScript)
