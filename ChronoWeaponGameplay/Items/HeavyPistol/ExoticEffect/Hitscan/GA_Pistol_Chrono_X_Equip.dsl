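; Event-triggered ability for the exotic Chrono heavy pistol's hitscan effect that creates an explosion at the shot's impact.
; If the source weapon is equipped, it takes the first blocking hit from event data and spawns an explosion BGA aligned to the normal, without checking hit validity.
; An async targeting preset runs from that actor, feeding an async GameplayEffect application throttled per tick.
; The ability ends without awaiting that application.
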(event Ability|EventActivateAbilityFromEvent
  (params (EventData :type Gameplay%20Event%20Data%20Structure%20%28by%20ref%29))
  (bind _returnvalue (GameInstanceSubsystems|GetTargetingSubsystem))
  (if (FortWeapon|IsEquipped (Ability|GetCurrentSourceWeapon))
    (Variables|Default|SetInstigator (Utilities|Struct|BreakGameplayEventData EventData))
    (bind (_valid _location _hitactor _normal) (CallFunction|GetHitInfo EventData))
    (bind _output_get (Variables|Default|SetHit%20Location _location))
    (bind _returnvalue_2 (Building|SpawnBuildingGameplayActor (Variables|Default|GetBGA%20Explosion) (Math|Transform|MakeTransform _output_get (Math|Rotator|MakeRotfromZ _normal)) (Variables|Default|GetInstigator) true false))
    (bind (_asynctaskproxy _targetinghandle) (Class|AsyncActionPerformTargeting|PerformTargetingAsyncAction _returnvalue_2 (Variables|Default|GetTargeting%20Preset))
      (:Targeted
        (if (not (Utilities|Array|IsEmpty (TargetingSystem|TargetingResults|GetTargetingResults _returnvalue _targetinghandle)))
          (Ability|Tasks|ApplyGameplayEffectToHitResultsAsync (Ability|GetAbilitySystemComponent (Variables|Default|GetInstigator)) (TargetingSystem|TargetingResults|GetTargetingResults _returnvalue _targetinghandle) (Variables|Default|GetGE_Explosion) 0 1 (Class|AthenaFunctionLibrary|MakeSimpleFortEffectContext (Variables|Default|GetInstigator) (Ability|GetCurrentSourceWeapon)) (Variables|Default|GetMax%20Targets%20to%20Process%20Per%20Tick%20CVar)
            (:OnCompleted
              _)))
        (Ability|EndAbility)))
    (else
      (Ability|EndAbility))))

(fn GetHitInfo
  (params (EventData :type Gameplay%20Event%20Data%20Structure))
  (bind (_eventtag _instigator _target _optionalobject _optionalobject2 _contexthandle _instigatortags _targettags _eventmagnitude _targetdata) (Utilities|Struct|BreakGameplayEventData EventData))
  (for _index (range (Ability|TargetData|GetDataCountfromTargetData _targetdata))
    (if (Collision|BreakHitResult (Ability|TargetData|GetHitResultfromTargetData _targetdata _index))
      (return true (.Location (Collision|BreakHitResult (Ability|TargetData|GetHitResultfromTargetData _targetdata _index))) (.HitActor (Collision|BreakHitResult (Ability|TargetData|GetHitResultfromTargetData _targetdata _index))) (.Normal (Collision|BreakHitResult (Ability|TargetData|GetHitResultfromTargetData _targetdata _index))))))
  (return false "0, 0, 0" 0 "0, 0, 0"))
